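; Audio cluster behavior that replaces individual BuzzCool clone voice lines with one shared sound at the cluster's position.
; Actors joining have their VO component silenced and flagged as clustered; leaving or stopping the cluster restores it.
; After a startup timer, the sound spawns at the offset cluster position, follows position updates and replays after random delays via OnAudioFinished.
; Stopping clears timers and triggers the sound's Stop parameter; death of all tracked actors stops playback.
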(event AudioClusterBehavior|EventOnActorAdded
  (params (Actor :type Actor%20Object%20Reference))
  ;[1 1 1] On Pawn added stop single VO audio.
  (Utilities|Array|AddUnique (Variables|Private|GetActors) Actor)
  (CallFunction|UpdateActorVOAudio Actor)
  (bind _pawn_buzz_cool_audio_component (Variables|Private|GetPawn%20Buzz%20Cool%20Audio%20Component)
    (:then
      (Default|BindEventtoOnDeath _pawn_buzz_cool_audio_component (EventDispatchers|CreateEvent :!event Custom|OnActorDeath)))
    (:else)))

(event AudioClusterBehavior|EventOnActorRemoved
  (params (Actor :type Actor%20Object%20Reference))
  ;[1 1 1] On Pawn removal start single vo audio back up.
  (bind _returnvalue (Utilities|Array|RemoveItem (Variables|Private|GetActors) Actor))
  (if _returnvalue
    (CallFunction|UpdateActorVOAudio Actor true)
    (bind _pawn_buzz_cool_audio_component (Variables|Private|GetPawn%20Buzz%20Cool%20Audio%20Component)
      (:then
        (Default|UnbindEventfromOnDeath _pawn_buzz_cool_audio_component (EventDispatchers|CreateEvent :!event Custom|OnActorDeath)))
      (:else))))

(event AudioClusterBehavior|EventOnPositionUpdated
  (params (NewPosition :type Vector%20%28by%20ref%29 :default "0, 0, 0"))
  ;[1 1 1] Update audio position
  (Variables|Private|SetNew%20Position (+ NewPosition (Variables|Default|GetOffset%20Position)))
  (bind _audiovoice (Variables|Private|GetAudioVoice)
    (:then
      (Transformation|SetWorldLocation _audiovoice (Variables|Private|GetNew%20Position) false false))
    (:else)))

(event AudioClusterBehavior|EventOnStop
  (CallFunction|StopTimer)
  (CallFunction|StopAudio))

(event AudioClusterBehavior|EventOnStart
  (params (InitialSize :type Integer))
  (bind _returnvalue (Utilities|Time|SetTimerbyEvent (delegate Custom|Start%20Sound) 1.0))
  (Variables|Private|SetStartAudioHandle _returnvalue))

(event Custom|Start%20Sound
  (CallFunction|OnAudioCall))

(event Custom|OnAudioCall
  ;[1 1 1] Play Audio - For none Cluster Pawns
  ;;[1 1 1] Attempt to play audio again if allowed, once the current line stops.
  (CallFunction|UpdateAudio true))

(event Custom|UpdateAudio
  (params (Play :type Boolean))
  ;[1 1 1] Play Audio - For none Cluster Pawns
  ;;[1 1 1] Play single audio line at random time once, but stop if clone enters audio cluster zone.
  (if Play
    (if (not (Variables|Private|GetbShouldStop))
      (bind _returnvalue (Utilities|Time|SetTimerbyEvent (EventDispatchers|CreateEvent :!event Custom|PlayAudio) (Math|Random|RandomFloatinRange 1.0 2.0)))
      (Variables|Private|SetHandlePlayAudio _returnvalue)
      (else
        _)))
  (CallFunction|StopTimer))

(event Custom|PlayAudio
  ;[1 1 1] Play Audio - For none Cluster Pawns
  ;;[1 1 1] Play audio and watch for when each audio triggers finish.
  (bind _returnvalue (Audio|SpawnSoundatLocation (Variables|Default|GetSound) (Variables|Private|GetNew%20Position)))
  (bind _output_get (Variables|Private|SetAudioVoice _returnvalue))
  (EventDispatchers|BindEventtoOnAudioFinished _output_get (EventDispatchers|CreateEvent :!event Custom|OnAudioCall)))

(event Custom|OnActorDeath
  (params (Pawn :type Actor%20Object%20Reference))
  (bind _actors (Variables|Private|GetActors))
  ;[1 1 1] Stop Sound playing at location if all actors die
  (bind _returnvalue (Utilities|Array|RemoveItem _actors Pawn))
  (if _returnvalue
    (if (Utilities|Array|IsEmpty _actors)
      (bind _audiovoice (Variables|Private|GetAudioVoice)
        (:then
          (Audio|Components|Audio|Stop _audiovoice))
        (:else)))))

(fn StopAudio
  (Variables|Private|SetbShouldStop true)
  (Utilities|FlowControl|Sequence
    (:then_0
      (bind _audiovoice (Variables|Private|GetAudioVoice)
        (:then
          (Audio|Parameter|ExecuteTriggerParameter _audiovoice "Stop"))
        (:else)))
    (:then_1
      (for _array_element (Variables|Private|GetActors)
        (Utilities|IsValid _array_element
          (:Is%20Valid
            (CallFunction|UpdateActorVOAudio _array_element true)))))))

(fn StopTimer
  (bind _startaudiohandle (Variables|Private|GetStartAudioHandle))
  (bind _handleplayaudio (Variables|Private|GetHandlePlayAudio))
  (Utilities|FlowControl|Sequence
    (:then_0
      (if (Utilities|Time|IsValidTimerHandle _handleplayaudio)
        (Utilities|Time|ClearandInvalidateTimerbyHandle _handleplayaudio)))
    (:then_1
      (if (Utilities|Time|IsValidTimerHandle _startaudiohandle)
        (Utilities|Time|ClearandInvalidateTimerbyHandle _startaudiohandle)))))

(fn UpdateActorVOAudio
  (params (ActorPawn :type Object%20Reference) (Play :type Boolean))
  (bind _aspawn_wild_estate_buzz_cool (Utilities|Casting|CastToPawn_WildEstate_BuzzCool ActorPawn))
  (Utilities|IsValid (Actor|GetComponentbyClass _aspawn_wild_estate_buzz_cool "/BuzzCool/Assets/Audio/Clones/BP_BuzzCool_Clone_Audio_Component.BP_BuzzCool_Clone_Audio_Component_C")
    (:Is%20Valid
      (bind _output_get (Variables|Private|SetPawn%20Buzz%20Cool%20Audio%20Component (Actor|GetComponentbyClass _aspawn_wild_estate_buzz_cool "/BuzzCool/Assets/Audio/Clones/BP_BuzzCool_Clone_Audio_Component.BP_BuzzCool_Clone_Audio_Component_C")))
      (Variables|/BuzzCool/Assets/Audio/Clones/BP_BuzzCool_Clone_Audio_Component.BP_BuzzCool_Clone_Audio_Component_C|SetIsAudioClustered (not Play) _output_get)
      (Class|buzzCoolAudioComponent|UpdateAudio _output_get Play))))
