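; Audio cluster behavior that swaps individual BuzzCool clone pawn voices for shared 2D audio driven by cluster size.
; Added actors have their clone audio component marked clustered and their voice stopped; removal or cluster stop restores it.
; On start it broadcasts itself to get the owner audio manager, then sets that manager's 2D audio "NumActors" parameter, suffixed by sector index, on size changes.
; Stopping zeroes the count and notifies the manager.
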(event AudioClusterBehavior|EventOnActorAdded
  (params (Actor :type Actor%20Object%20Reference))
  ;[1 1 1] On Pawn added stop single VO audio.
  (Utilities|Array|AddUnique (Variables|Private|GetActors) Actor)
  (CallFunction|UpdateActorVOAudio Actor))

(event AudioClusterBehavior|EventOnActorRemoved
  (params (Actor :type Actor%20Object%20Reference))
  ;[1 1 1] On Pawn removal start single vo audio back up.
  (bind _returnvalue (Utilities|Array|RemoveItem (Variables|Private|GetActors) Actor))
  (if _returnvalue
    (CallFunction|UpdateActorVOAudio Actor true)))

(event AudioClusterBehavior|EventOnStart
  (params (InitialSize :type Integer))
  ;[1 1 1] Send Cluster to owner audio manager
  (Variables|Private|SetClusterSize InitialSize)
  (GameplayEvent|BroadcastEvent (Variables|Private|GetClusterState) 0 (Utilities|Struct|MakeStructBuzzCoolAudioCluster self) "/Script/FortniteGame.FortGameState" (Fortnite|GetFirstLocalFortPlayerController true true))
  ;[1 1 1] Get cluster data and current index size.
  (bind _value (Utilities|InstancedStruct|GetInstancedStructValue (Variables|AudioClusterBehavior|GetClusterData))
    (:Valid
      (Variables|Private|SetClusterSectorIndex (Utilities|Struct|BreakStaticSectorAudioClusterData _value))
      (Default|UpdateGlobalSound (Variables|Private|GetClusterSize)))))

(event AudioClusterBehavior|EventOnStop
  (bind _output_get (Variables|Private|SetClusterSize 0))
  (Default|UpdateGlobalSound _output_get)
  (for _array_element (Variables|Private|GetActors)
    (Utilities|IsValid _array_element
      (:Is%20Valid
        (CallFunction|UpdateActorVOAudio _array_element true))))
  (bind _owner_audio_manger (Variables|Private|GetOwner%20Audio%20Manger)
    (:then
      (Class|BPBuzzCoolOwnerAudioManager|AudioClusterDestroyed _owner_audio_manger self))
    (:else)))

(event AudioClusterBehavior|EventOnSizeUpdated
  (params (NewSize :type Integer))
  (bind _output_get (Variables|Private|SetClusterSize NewSize))
  (Default|UpdateGlobalSound _output_get))

(event Custom|Update%20Owner%20Audio%20Manager
  (params (Owner%20Audio%20Manger :type BP%20Buzz%20Cool%20Owner%20Audio%20Manager%20Object%20Reference))
  ;[1 1 1] Recieve update from Owner Audio Manager (OnStart)
  (Variables|Private|SetOwner%20Audio%20Manger Owner%20Audio%20Manger)
  (Default|UpdateGlobalSound (Variables|Private|GetClusterSize)))

(fn UpdateActorVOAudio
  (params (ActorPawn :type Object%20Reference) (Play :type Boolean))
  ;[1 1 1] Update AI pawn audio
  (bind _aspawn_wild_estate_buzz_cool (Utilities|Casting|CastToPawn_WildEstate_BuzzCool ActorPawn))
  (Utilities|IsValid (Actor|GetComponentbyClass _aspawn_wild_estate_buzz_cool "/BuzzCool/Assets/Audio/Clones/BP_BuzzCool_Clone_Audio_Component.BP_BuzzCool_Clone_Audio_Component_C")
    (:Is%20Valid
      (bind _output_get (Variables|Private|SetPawn%20Buzz%20Cool%20Audio%20Component (Actor|GetComponentbyClass _aspawn_wild_estate_buzz_cool "/BuzzCool/Assets/Audio/Clones/BP_BuzzCool_Clone_Audio_Component.BP_BuzzCool_Clone_Audio_Component_C")))
      ;;[1 1 1] Start/Stop single audio voice that is on pawn when being included in the cluster.
      (Variables|/BuzzCool/Assets/Audio/Clones/BP_BuzzCool_Clone_Audio_Component.BP_BuzzCool_Clone_Audio_Component_C|SetIsAudioClustered (not Play) _output_get)
      (Class|buzzCoolAudioComponent|UpdateAudio _output_get Play))))

(fn UpdateGlobalSound
  (params (NewSize :type Integer))
  ;[1 1 1] Update audio component size index.
  (bind _owner_audio_manger (Variables|Private|GetOwner%20Audio%20Manger)
    (:then
      (bind _2d_audio (Variables|/BuzzCool/Assets/Audio/Clones/BP_BuzzCool_Owner_Audio_Manager.BP_BuzzCool_Owner_Audio_Manager_C|Get2D%20Audio _owner_audio_manger)
        (:then
          (Audio|Parameter|SetFloatParameter _2d_audio (Utilities|String|StringToName (Utilities|String|Append "NumActors" (Utilities|String|ToString(Integer) (Variables|Private|GetClusterSectorIndex)))) (Math|Conversions|ToFloat(Integer) NewSize) :!class /Script/AudioExtensions.AudioParameterControllerInterface))
        (:else)))
    (:else)))
